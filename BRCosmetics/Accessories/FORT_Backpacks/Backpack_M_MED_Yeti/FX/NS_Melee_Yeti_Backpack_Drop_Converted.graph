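# Cascade-converted Niagara effect for the Yeti backpack cosmetic, showing blue water drips and twinkling stars as two looping sprite emitters.
# Water_drop spawns short-lived drops at random offsets, inherits horizontal velocity, then applies strong downward acceleration, drag and a fade-in/fade-out curve.
# Stars spawn on a spherical shape with slight velocity inheritance, flickering through an oscillating size curve, boosted RGB, alpha fade and random plus animated dynamic material parameters.
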NiagaraSystem /BRCosmetics/Accessories/FORT_Backpacks/Backpack_M_MED_Yeti/FX/NS_Melee_Yeti_Backpack_Drop_Converted
dependencies {usedDataInterfaces [&NiagaraDataInterfaceCurve], usedDynamicInputs [], usedModules [&SystemState]}
emitters {
 Water_drop {
  emitterUpdateScript [
   &EmitterState(
    Life Cycle Mode=::NewEnumerator1
    Loop Delay=0.3
    Loop Duration=&RandomRangeFloat(Minimum=0, Maximum=1, Evaluation Type=::NewEnumerator1)
    UseLoopDelay=true
   )
   &SpawnRate(SpawnRate=3, Use Spawn Probability=false)
  ]
  particleSpawnScript [
   &InitializeParticle(
    Lifetime=0.4
    Sprite Size=&RandomRangeVector2D(Minimum=<3 8>, Maximum=<4.5 9>, Evaluation Type=::NewEnumerator1)
    Sprite Size Mode=::NewEnumerator4
    UsePositionOffset=false
    Write Material Random=true
   )
   &Color(Color=&MakeLinearColorFromVectorAndFloat("Vector (RGB)"=<0.2142 0.4953 1>, "Float (Alpha)"=1))
   /CascadeToNiagaraConverter/NiagaraScripts/ModuleScripts/AddVectorToPosition(Vector:Vector3f=&TransformVector(
     Vector=&RandomRangeVector(Minimum=<-5 -15 -9>, Maximum=<-5 -8 10>, Evaluation Type=::NewEnumerator1)
     Destination Space=::Simulation
    ))
   &InheritVelocity(Inherited Velocity Amount Scale=<1 1 0>, Inherited Velocity Speed Limit=0, Speed Limit=false, UseSpeedThreshold=false)
  ]
  particleUpdateScript [
   &ParticleState()
   &ScaleSpriteSize(
    Scale Sprite Size Mode=::NewEnumerator0
    Uniform Scale Factor=&FloatFromCurve(FloatCurve=DataInterface() {propertyValues {Curve {keys [
         {arriveTangent=-0.0075, time=-0.0009, value=0.8285}
         {arriveTangent=-0.8277, interpMode=RCIM_Cubic, leaveTangent=-0.8277, time=0.6547, value=0.821}
         <at 1 0>
        ]}}})
   )
   &AccelerationForce(Acceleration=<0 0 -800>)
   &Drag(Drag=3, Use Linear Drag=true, Use Rotational Drag=true)
   &SolveForcesAndVelocity(Clamp Velocity=false, Limit Acceleration=false)
   &ScaleColor(
    Scale Alpha=&FloatFromCurve(FloatCurve=DataInterface() {propertyValues {Curve {keys [
         <at 0.1485 0.001 RCIM_Cubic>
         {arriveTangent=3.4768, interpMode=RCIM_Cubic, leaveTangent=3.4768, time=0.3246, value=0.0372}
         {arriveTangent=-0.8051, interpMode=RCIM_Cubic, leaveTangent=-0.5394, tangentMode=RCTM_Break, time=0.3788, value=0.8018}
         {arriveTangent=-4.1559, interpMode=RCIM_Cubic, tangentMode=RCTM_Break, time=1, value=0}
        ]}}})
    ScaleA=true
    ScaleRGB=true
   )
  ]
  rendererClasses [&NiagaraSpriteRendererProperties]
  renderers [{rendererClass=&NiagaraSpriteRendererProperties}]
 }
 stars {
  emitterUpdateScript [
   &EmitterState(Life Cycle Mode=::NewEnumerator1, Loop Duration=0.3, UseLoopDelay=true)
   &SpawnRate(SpawnRate=2, Use Spawn Probability=false)
  ]
  particleSpawnScript [
   &InitializeParticle(
    Lifetime=0.4
    Sprite Size=&Vector2DFromFloat(Value=&RandomRangeFloat(Minimum=1.6, Maximum=4, Evaluation Type=::NewEnumerator1))
    Sprite Size Mode=::NewEnumerator4
    UsePositionOffset=false
    Write Material Random=true
   )
   &DynamicMaterialParameters(
    Index 0 Param 1=&RandomRangeFloat(Minimum=0, Maximum=0.1, Evaluation Type=::NewEnumerator1)
    Index 0 Param 4=&RandomRangeFloat(Minimum=0, Maximum=0.5, Evaluation Type=::NewEnumerator1)
    Module.Param0WriteEnabled=true
    Module.Param1WriteEnabled=false
    Module.Param2WriteEnabled=false
    Module.Param3WriteEnabled=true
   )
   &Color(Color=&MakeLinearColorFromVectorAndFloat("Vector (RGB)"=<0.1958 0.5098 1>, "Float (Alpha)"=1))
   &InheritVelocity(Inherited Velocity Amount Scale=<0.05 0.05 0.05>, Inherited Velocity Speed Limit=0, Speed Limit=false, UseSpeedThreshold=false)
   &ShapeLocation(
    Hemisphere Distribution=<0.5 1>
    MaskSpawnGroup=false
    Override Seed=false
    Rotation Mode=::NewEnumerator2
    Sphere Radius=21
    Sphere Surface Distribution=1
    UseOwnerScale=true
    Yaw / Pitch / Roll=<0.5 0 0>
   )
   &AddVectorToPosition(Vector:Vector3f=&TransformVector(Vector=<3 -5 0>, Destination Space=::Simulation))
  ]
  particleUpdateScript [
   &ParticleState()
   &ScaleSpriteSize(
    Scale Sprite Size Mode=::NewEnumerator0
    Uniform Scale Factor=&FloatFromCurve(FloatCurve=DataInterface() {propertyValues {Curve {keys [
         {arriveTangent=27.05, interpMode=RCIM_Cubic, leaveTangent=27.05, tangentMode=RCTM_User, time=0, value=0}
         {arriveTangent=1.3101, interpMode=RCIM_Cubic, leaveTangent=1.3101, time=0.0546, value=1.2477}
         {arriveTangent=-1.3446, interpMode=RCIM_Cubic, leaveTangent=-1.3446, time=0.1581, value=0.2072}
         {arriveTangent=0.1974, interpMode=RCIM_Cubic, leaveTangent=0.1974, time=0.2882, value=0.9336}
         {arriveTangent=0.2295, interpMode=RCIM_Cubic, leaveTangent=0.2295, time=0.4301, value=0.2608}
         {arriveTangent=-0.3367, interpMode=RCIM_Cubic, leaveTangent=-0.3367, time=0.5884, value=1.0025}
         {arriveTangent=-1.2837, interpMode=RCIM_Cubic, leaveTangent=-1.2837, time=0.7448, value=0.1549}
         {arriveTangent=-0.607, interpMode=RCIM_Cubic, leaveTangent=-0.607, time=0.834, value=0.6872}
         <at 1 0 RCIM_Cubic>
        ]}}})
   )
   &DynamicMaterialParameters(
    Index 0 Param 2=&FloatFromCurve(FloatCurve=DataInterface() {propertyValues {Curve {keys [<at 0.6 1 RCIM_Cubic>,<at 1 0 RCIM_Cubic>]}}})
    Index 0 Param 3=&FloatFromCurve(FloatCurve=DataInterface() {propertyValues {Curve {keys [<at 0 0>,<at 0.5 20>]}}})
    Module.Param0WriteEnabled=false
    Module.Param1WriteEnabled=true
    Module.Param2WriteEnabled=true
    Module.Param3WriteEnabled=false
   ) {module=DynamicMaterialParameters001}
   &ScaleColor(
    Scale Alpha=&FloatFromCurve(FloatCurve=DataInterface() {propertyValues {Curve {keys [<at 0 1>,<at 0.6 1>,<at 1 0>]}}})
    Scale RGB=<20 20 20>
    ScaleA=true
    ScaleRGB=true
   )
   &Drag(Drag=7, Use Linear Drag=true, Use Rotational Drag=true)
   &AccelerationForce(Acceleration=<0 0 -10>)
   &SolveForcesAndVelocity(Clamp Velocity=false, Limit Acceleration=false)
  ]
  rendererClasses [&NiagaraSpriteRendererProperties]
  renderers [{rendererClass=&NiagaraSpriteRendererProperties}]
 }
}
summary {
 systemName=NS_Melee_Yeti_Backpack_Drop_Converted
 emitters [
  {emitterName=Water_drop, rendererClasses [&NiagaraSpriteRendererProperties]}
  {emitterName=stars, rendererClasses [&NiagaraSpriteRendererProperties]}
 ]
 userVariables []
}
