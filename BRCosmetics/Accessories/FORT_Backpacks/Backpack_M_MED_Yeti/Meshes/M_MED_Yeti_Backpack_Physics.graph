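# Physics asset for the skeletal mesh of the Yeti backpack cosmetic, defining simple bodies for a short dangling chain.
# A capsule on backpack_root anchors two smaller, slightly offset capsules on dyn_chain_01 and dyn_chain_02, all zero-length and so effectively spheres.
# Constraints link the bodies in series from the root, with moderate swing limits and tight twist limits that let the chain sway with little twisting.
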PhysicsAsset /BRCosmetics/Accessories/FORT_Backpacks/Backpack_M_MED_Yeti/Meshes/M_MED_Yeti_Backpack_Physics
bodies {
 backpack_root [{shapeType=Capsule, length=0, radius=3}]
 dyn_chain_01 [{shapeType=Capsule, length=0, radius=1, center=<1 0 0>}]
 dyn_chain_02 [{shapeType=Capsule, length=0, radius=1, center=<1 0 0>}]
}
constraints [
 {bone1Name=dyn_chain_01, bone2Name=backpack_root, swing1LimitDegrees=22.5, swing2LimitDegrees=22.5, twistLimitDegrees=5}
 {bone1Name=dyn_chain_02, bone2Name=dyn_chain_01, swing1LimitDegrees=22.5, swing2LimitDegrees=22.5, twistLimitDegrees=5}
]
